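# Niagara effect for the PearlTaffyOpera emote, adding pink and blue stage-like ambience and glowing hand trails.
# Environment multiplexes spawn groups through visibility tags, feeding sprite, mesh and light renderers for swirling particles, heart glows and spotlights.
# Its particle lights switch off in front-end preview.
# HandTrails and TrailSparkles sample two skeletal bones or sockets, scale spawning by User.HandTrails, and switch world/local position blending on an owner-velocity threshold.
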NiagaraSystem /BRCosmetics/Animation/Game/MainPlayer/Emotes/PearlTaffyOpera/FX/NS_Emote_PearlTaffyOpera
dependencies {
 usedDataInterfaces [
  &NiagaraDataInterfaceCurve
  &NiagaraDataInterfaceArrayColor
  &NiagaraDataInterfaceColorCurve
  &NiagaraDataInterfaceSkeletalMesh
 ]
 usedDynamicInputs []
 usedModules [
  &self:SystemSpawnScript.SystemScriptSource.SystemScriptGraph.NiagaraNodeAssignment_0.SetVariables_#1
  &SystemState
  &BlendSpaceHelper
 ]
}
emitters {
 Environment {
  emitterUpdateScript [
   &EmitterState()
   &SpawnBurst_Instantaneous(Spawn Count=2, Spawn Group=5, Use Spawn Probability=false, UseLoopCountLimit=false) {module=SpawnBurst_Instantaneous004}
   &SpawnBurst_Instantaneous(Spawn Count=12, Spawn Time=0.16, Use Spawn Probability=false, UseLoopCountLimit=false)
   &SpawnBurst_Instantaneous(Spawn Count=3, Spawn Group=1, Spawn Time=0.16, Use Spawn Probability=false, UseLoopCountLimit=false) {module=SpawnBurst_Instantaneous001}
   &SpawnBurst_Instantaneous(Spawn Count=36, Spawn Group=2, Spawn Time=0.16, Use Spawn Probability=false, UseLoopCountLimit=false) {module=SpawnBurst_Instantaneous002}
   &SpawnBurst_Instantaneous(Spawn Count=2, Spawn Group=4, Spawn Time=0.08, Use Spawn Probability=false, UseLoopCountLimit=false) {module=SpawnBurst_Instantaneous003}
   &SpawnRate(
    SpawnRate=&FloatFromCurve(
     FloatCurve=DataInterface() {propertyValues {Curve {keys [
         <at 0.075 0.5 RCIM_Cubic>
         {arriveTangent=-0.2556, interpMode=RCIM_Cubic, leaveTangent=-0.2556, tangentMode=RCTM_User, time=0.1, value=1}
         {arriveTangent=0.002, interpMode=RCIM_Cubic, leaveTangent=0.002, tangentMode=RCTM_User, time=0.3, value=1}
         <at 1 0.3 RCIM_Cubic>
        ]}}}
     CurveIndex=$Emitter.NormalizedLoopAge
     Scale Curve=4.5
    )
    Use Spawn Probability=false
   )
   &SpawnRate(
    Spawn Probability=0.2
    SpawnGroup=1
    SpawnRate=&FloatFromCurve(
     FloatCurve=DataInterface() {propertyValues {Curve {keys [
         <at 0.075 0.5 RCIM_Cubic>
         {arriveTangent=-0.2556, interpMode=RCIM_Cubic, leaveTangent=-0.2556, tangentMode=RCTM_User, time=0.1, value=1}
         {arriveTangent=0.002, interpMode=RCIM_Cubic, leaveTangent=0.002, tangentMode=RCTM_User, time=0.3, value=1}
         <at 1 0.3 RCIM_Cubic>
        ]}}}
     CurveIndex=$Emitter.NormalizedLoopAge
     Scale Curve=6
    )
    Use Spawn Probability=true
   ) {module=SpawnRate001}
   &SpawnRate(
    SpawnGroup=2
    SpawnRate=&FloatFromCurve(
     FloatCurve=DataInterface() {propertyValues {Curve {keys [
         <at 0.075 0.5 RCIM_Cubic>
         {arriveTangent=-0.2556, interpMode=RCIM_Cubic, leaveTangent=-0.2556, tangentMode=RCTM_User, time=0.1, value=1}
         {arriveTangent=0.002, interpMode=RCIM_Cubic, leaveTangent=0.002, tangentMode=RCTM_User, time=0.3, value=1}
         <at 1 0.3 RCIM_Cubic>
        ]}}}
     CurveIndex=$Emitter.NormalizedLoopAge
     Scale Curve=20
    )
    Use Spawn Probability=false
   ) {module=SpawnRate002}
   &SpawnRate(
    Spawn Probability=0.8
    SpawnGroup=3
    SpawnRate=&FloatFromCurve(
     FloatCurve=DataInterface() {propertyValues {Curve {keys [
         {arriveTangent=1.25, interpMode=RCIM_Cubic, time=0, value=1}
         {arriveTangent=-1, interpMode=RCIM_Cubic, leaveTangent=-1, time=0.8, value=1}
         <at 1 0 RCIM_Cubic>
        ]}}}
     CurveIndex=$Emitter.NormalizedLoopAge
     Scale Curve=2.45
    )
    Use Spawn Probability=true
   ) {module=SpawnRate003}
   /SetVariables(
    Emitter.BlueColor_Inner:LinearColor=&RandomRangeLinearColor(Minimum=<rgba 0.2057 0.415 0.6094 1>, Maximum=<rgba 0.1423 0.6094 0.551 1>)
    Emitter.BlueColor_Outer:LinearColor=&RandomRangeLinearColor(Minimum=<rgba 0.1053 0.5825 1 1>, Maximum=<rgba 0.522 1 0.8743 1>)
    Emitter.PinkColor_Inner:LinearColor=&RandomRangeLinearColor(Minimum=<rgba 0.5938 0.0652 0.1971 1>, Maximum=<rgba 0.5938 0.0938 0.4044 1>)
    Emitter.PinkColor_Outer:LinearColor=&RandomRangeLinearColor(Minimum=<rgba 1 0.2188 0.3698 1>, Maximum=<rgba 1 0.1472 0.868 1>)
   ) {isSetParameters=true}
  ]
  particleSpawnScript [
   /SetVariables(Particles.VisibilityTag=&SpawnGroup()) {isSetParameters=true}
   &InitializeParticle(
    Color=&MergeHelper_ColorSwitch(
     0=<rgba 1 1 1 1>
     1=<rgba 1 1 1 1>
     2=&SelectLinearColorFromArray(
      Array Sampling Mode=::NewEnumerator0
      Direct Array Index=?
      Color Selection Array=DataInterface() {propertyValues {ColorData [<rgba 3 0.7253 2.2894 1>, <rgba 2.9639 0.3542 3 1>, <rgba 3 0.3578 0.9064 1>, <rgba 2.2268 0.8787 3 1>, <rgba 0.8923 2.8279 3 1>]}}
      Random Seed=0
     )
     3=&RandomRangeLinearColor(Minimum=<rgba 1 0.272 0.9128 0.2>, Maximum=<rgba 1 0.1935 0.4262 0.55>)
     4=<rgba 1 1 1 1>
     Index=&SpawnGroup()
     Num=::NewEnumerator3
    )
    Lifetime=&MergeHelper_FloatSwitch(
     0=&RandomRangeFloat(Minimum=1.2, Maximum=2.6)
     1=&RandomRangeFloat(Minimum=1.2, Maximum=2.6)
     2=&RandomRangeFloat(Minimum=0.8, Maximum=3.6)
     3=&RandomRangeFloat(Minimum=1, Maximum=2)
     4=11.637
     5=11.637
     Index=&SpawnGroup()
     Num=::NewEnumerator4
    )
    Mesh Scale Mode=::NewEnumerator3
    Sprite Rotation Mode=::NewEnumerator2
    Sprite Size Mode=::NewEnumerator3
    Uniform Sprite Size=&MergeHelper_FloatSwitch(
     0=&UniformAOrBFloat(A=12, B=18, Distribution Weight=0.2, Evaluation Type=::NewEnumerator0, Recalculate Random Each Loop=true, Override Seed=false)
     1=&UniformAOrBFloat(A=12, B=18, Distribution Weight=0.2, Evaluation Type=::NewEnumerator0, Recalculate Random Each Loop=true, Override Seed=false)
     2=&RandomRangeFloat(Minimum=5, Maximum=12)
     3=&UniformAOrBFloat(A=15, B=32, Distribution Weight=0.725, Evaluation Type=::NewEnumerator0, Recalculate Random Each Loop=true, Override Seed=false)
     Index=&SpawnGroup()
     Num=::NewEnumerator2
    )
    UsePositionOffset=false
    Write Material Random=true
   )
   &ShapeLocation(
    MaskSpawnGroup=false
    Non Uniform Scale=&MaskVectorBySpawnGroup(Masked Vector=<1 1 1>, Passthrough Vector=<1 1 1.3>, Spawn Group Mask=3)
    Offset=&MergeHelper_VectorSwitch(0=<0 0 120>, 1=<0 0 120>, 2=<0 0 120>, 3=<0 0 135>, Index=&SpawnGroup(), Num=::NewEnumerator2)
    Offset Mode=::NewEnumerator0
    Override Seed=false
    Sphere Radius=&MergeHelper_FloatSwitch(0=65, 1=65, 2=65, 3=80, Index=&SpawnGroup(), Num=::NewEnumerator2)
    UseOwnerScale=true
   )
   &CameraOffset(Camera Offset Amount=&MergeHelper_FloatSwitch(
     0=0
     1=0
     2=0
     3=&RandomRangeFloat(Minimum=-10, Maximum=-25)
     4=0
     Index=&SpawnGroup()
     Num=::NewEnumerator3
    ))
  ]
  particleUpdateScript [
   &ParticleState()
   &SpriteRotationRate(Rotation Rate=&RandomRangeFloat(Minimum=-100, Maximum=100))
   &GravityForce(Gravity=&MakeVector(
     X=0
     Y=0
     Z=&MergeHelper_FloatSwitch(
      0=&RandomRangeFloat(Minimum=-43.75, Maximum=-87.5)
      1=&RandomRangeFloat(Minimum=-43.75, Maximum=-87.5)
      2=&RandomRangeFloat(Minimum=-78.75, Maximum=-140)
      3=0
      4=0
      Index=&SpawnGroup()
      Num=::NewEnumerator3
     )
    ))
   &CurlNoiseForce(
    Bias Noise=false
    Noise Frequency=&MergeHelper_FloatSwitch(
     0=&RandomRangeFloat(Minimum=25, Maximum=36)
     1=&RandomRangeFloat(Minimum=12, Maximum=20)
     2=&RandomRangeFloat(Minimum=20, Maximum=55)
     3=&RandomRangeFloat(Minimum=20, Maximum=55)
     4=0
     Index=&SpawnGroup()
     Num=::NewEnumerator3
    )
    Noise Strength=&MergeHelper_FloatSwitch(
     0=&FloatFromCurve(
      FloatCurve=DataInterface() {propertyValues {Curve {defaultValue=1, keys [<at 0 0 RCIM_Cubic>,<at 0.2 1 RCIM_Cubic>]}}}
      Scale Curve=&RandomRangeFloat(Minimum=180, Maximum=250)
     )
     1=&FloatFromCurve(
      FloatCurve=DataInterface() {propertyValues {Curve {defaultValue=1, keys [<at 0 0 RCIM_Cubic>,<at 0.2 1 RCIM_Cubic>]}}}
      Scale Curve=&RandomRangeFloat(Minimum=180, Maximum=250)
     )
     2=&RandomRangeFloat(Minimum=70, Maximum=250)
     3=10
     4=0
     Index=&SpawnGroup()
     Num=::NewEnumerator3
    )
    Pan Noise=false
   )
   &Drag(Drag=&RandomRangeFloat(Minimum=1.6, Maximum=3.2), Use Linear Drag=true, Use Rotational Drag=true)
   &VortexForce(
    Origin Pull Amount:NiagaraFloat=&MergeHelper_FloatSwitch(0=120, 1=120, 2=200, 3=0, Index=&SpawnGroup(), Num=::NewEnumerator2)
    UseOriginOffset=false
    Vortex Force Amount=&MergeHelper_FloatSwitch(
     0=&RandomRangeFloat(Minimum=280, Maximum=425)
     1=&RandomRangeFloat(Minimum=280, Maximum=425)
     2=600
     3=0
     4=0
     Index=&SpawnGroup()
     Num=::NewEnumerator3
    )
   )
   &SolveForcesAndVelocity(Clamp Velocity=false, Limit Acceleration=false)
   &ScaleColor(
    Scale Alpha=&FloatFromCurve(
     FloatCurve=DataInterface() {propertyValues {Curve {
        defaultValue=1
        keys [
         {arriveTangent=9.0725, interpMode=RCIM_Cubic, leaveTangent=9.0725, tangentMode=RCTM_User, time=0, value=0}
         {arriveTangent=-0.0478, interpMode=RCIM_Cubic, leaveTangent=-0.0478, tangentMode=RCTM_User, time=0.1408, value=0.9847}
         {arriveTangent=-1.4132, interpMode=RCIM_Cubic, leaveTangent=-1.4132, tangentMode=RCTM_User, time=1, value=0}
        ]
       }}}
     Scale Curve=&MergeHelper_FloatSwitch(
      0=0.4
      1=0.4
      2=&Waveform(
       "Waveform [1]"=::NewEnumerator2
       "[1] Frequency"=&RandomRangeFloat(Minimum=0.4, Maximum=1.2)
       "[1] Compound Frequency"=4
       "[1] Amplitude Min/Max"=<0 2>
      )
      3=?
      Index=&SpawnGroup()
      Num=::NewEnumerator1
     )
    )
    ScaleA=true
    ScaleRGB=true
   )
   /SetVariables(
    Particles.Glow Color=&MultiplyLinearColor(
     Linear Color=$Particles.Color
     Scale Factor=&MakeVector4FromVectorAndFloat("Vector (RGB)"=<1 1 1>, "Float (Alpha)"=0.05)
    )
    Particles.Glow Size=&Multiply_Vector2D_ByFloat(Float=1.8, Vector2D=$Particles.SpriteSize)
   ) {isSetParameters=true}
   /SetVariables(
    Particles.Spotlight1Color:LinearColor=&MakeLinearColorFromVectorAndFloat(
     "Vector (RGB)"=&MakeVectorFromLinearColorRGB(Color=&ColorFromCurve(ColorCurve=DataInterface() {propertyValues {
         AlphaCurve {keys []}
         BlueCurve {keys [<at 0 0.9911 RCIM_Cubic>,<at 0.4876 0.5828>,<at 1 0.9911>]}
         GreenCurve {keys [<at 0 0.521 RCIM_Cubic>,<at 0.4876 0.2974>,<at 1 0.521>]}
         RedCurve {keys [<at 0 1 RCIM_Cubic>,<at 0.4876 1>,<at 1 1>]}
        }}))
     "Float (Alpha)"=&FloatFromCurve(
      FloatCurve=DataInterface() {propertyValues {Curve {keys [
          <at 0 0 RCIM_Cubic>
          {arriveTangent=-0.0358, interpMode=RCIM_Cubic, leaveTangent=-0.0358, tangentMode=RCTM_User, time=0.125, value=1}
          {arriveTangent=-0.0388, interpMode=RCIM_Cubic, leaveTangent=-0.0388, tangentMode=RCTM_User, time=0.9, value=1}
          <at 1 0 RCIM_Cubic>
         ]}}}
      Scale Curve=&Waveform("Waveform [1]"=::NewEnumerator0, "[1] Frequency"=1, "[1] Amplitude Min/Max"=<0.125 0.25>)
     )
    )
    Particles.Spotlight1Orientation:Quat4f=<0 -0.15 0 1>
    Particles.Spotlight1Pos:NiagaraPosition=&AddVectorToPosition(Position=/Niagara/DynamicInputs/Helpers/SimulationPosition(), Vector=<-75 0 220>)
    Particles.Spotlight2Orientation:Quat4f=<0 0.15 0 1>
    Particles.Spotlight2Pos:NiagaraPosition=&AddVectorToPosition(Position=&SimulationPosition(), Vector=<75 0 220>)
   ) {isSetParameters=true}
   &FN_Light_Attributes(
    Color=<rgba 1 0.3593 0.9553 1>
    Create Color Attribute=true
    Create Exponent Attribute=true
    Create Position Attribute=true
    Create Radius Attribute=true
    Enable Particle Light Enable=true
    Exponent=4
    Light Enabled:NiagaraBool=&InvertBool(Boolean:NiagaraBool=$User.bIsFrontEndPreview)
    Position=<0 45 145>
    Radius=120
    Scale Color=&FloatFromCurve(
     FloatCurve=DataInterface() {propertyValues {Curve {keys [
         <at 0 0 RCIM_Cubic>
         {arriveTangent=7.6936, interpMode=RCIM_Cubic, leaveTangent=7.6936, tangentMode=RCTM_User, time=0.02, value=1.4}
         {arriveTangent=0.0613, interpMode=RCIM_Cubic, leaveTangent=0.0613, tangentMode=RCTM_User, time=0.1, value=1}
         {arriveTangent=-0.0388, interpMode=RCIM_Cubic, leaveTangent=-0.0388, tangentMode=RCTM_User, time=0.9, value=1}
         <at 1 0 RCIM_Cubic>
        ]}}}
     Scale Curve=&Waveform("Waveform [1]"=::NewEnumerator2, "[1] Frequency"=1, "[1] Amplitude Min/Max"=<3 5>)
    )
   )
   &FN_Light_Attributes(
    Color=<rgba 1 0.3593 0.9553 1>
    Create Color Attribute=true
    Create Exponent Attribute=true
    Create Position Attribute=true
    Create Radius Attribute=true
    Enable Particle Light Enable=true
    Exponent=2
    Light Enabled:NiagaraBool=&InvertBool(Boolean:NiagaraBool=$User.bIsFrontEndPreview)
    Position=<0 0 15>
    Radius=&FloatFromCurve(
     FloatCurve=DataInterface() {propertyValues {Curve {keys [
         <at 0 0.4 RCIM_Cubic>
         {arriveTangent=7.6936, interpMode=RCIM_Cubic, leaveTangent=7.6936, tangentMode=RCTM_User, time=0.02, value=1.4}
         {arriveTangent=0.0613, interpMode=RCIM_Cubic, leaveTangent=0.0613, tangentMode=RCTM_User, time=0.035, value=1}
        ]}}}
     Scale Curve=140
    )
    Scale Color=&FloatFromCurve(
     FloatCurve=DataInterface() {propertyValues {Curve {keys [
         <at 0 0 RCIM_Cubic>
         {arriveTangent=7.6936, interpMode=RCIM_Cubic, leaveTangent=7.6936, tangentMode=RCTM_User, time=0.02, value=1.2}
         {arriveTangent=0.0613, interpMode=RCIM_Cubic, leaveTangent=0.0613, tangentMode=RCTM_User, time=0.1, value=1}
         {arriveTangent=-0.0388, interpMode=RCIM_Cubic, leaveTangent=-0.0388, tangentMode=RCTM_User, time=0.9, value=1}
         <at 1 0 RCIM_Cubic>
        ]}}}
     Scale Curve=&Waveform("Waveform [1]"=::NewEnumerator2, "[1] Frequency"=1, "[1] Amplitude Min/Max"=<8 12>)
    )
   ) {module=FN_Light_Attributes001}
   /SetVariables(
    Particles.BlueHeartGlowColor:LinearColor=&MultiplyLinearColor(
     Linear Color=<rgba 0.1562 2.6291 3 0.2>
     Scale Factor=&MakeVector4FromVectorAndFloat(
      "Vector (RGB)"=<1 1 1>
      "Float (Alpha)"=&FloatFromCurve(
       FloatCurve=DataInterface() {propertyValues {Curve {
          defaultValue=1
          keys [
           {arriveTangent=9.0725, interpMode=RCIM_Cubic, leaveTangent=9.0725, tangentMode=RCTM_User, time=0, value=0}
           {arriveTangent=-0.0478, interpMode=RCIM_Cubic, leaveTangent=-0.0478, tangentMode=RCTM_User, time=0.1408, value=0.9847}
           {arriveTangent=-1.4132, interpMode=RCIM_Cubic, leaveTangent=-1.4132, tangentMode=RCTM_User, time=1, value=0}
          ]
         }}}
       Scale Curve=&Waveform("Waveform [1]"=::NewEnumerator0, "[1] Frequency"=1, "[1] Amplitude Min/Max"=<0.4 1>)
      )
     )
    )
    Particles.HeartGlowSize:Vector2f=&Multiply_Vector2D_ByFloat(Float=1.45, Vector2D=$Particles.SpriteSize)
    Particles.PinkHeartGlowColor:LinearColor=&MultiplyLinearColor(
     Linear Color=<rgba 3 0.3125 2.6262 0.2>
     Scale Factor=&MakeVector4FromVectorAndFloat(
      "Vector (RGB)"=<1 1 1>
      "Float (Alpha)"=&FloatFromCurve(
       FloatCurve=DataInterface() {propertyValues {Curve {
          defaultValue=1
          keys [
           {arriveTangent=9.0725, interpMode=RCIM_Cubic, leaveTangent=9.0725, tangentMode=RCTM_User, time=0, value=0}
           {arriveTangent=-0.0478, interpMode=RCIM_Cubic, leaveTangent=-0.0478, tangentMode=RCTM_User, time=0.1408, value=0.9847}
           {arriveTangent=-1.4132, interpMode=RCIM_Cubic, leaveTangent=-1.4132, tangentMode=RCTM_User, time=1, value=0}
          ]
         }}}
       Scale Curve=&Waveform("Waveform [1]"=::NewEnumerator0, "[1] Frequency"=1, "[1] Amplitude Min/Max"=<0.4 1>)
      )
     )
    )
   ) {isSetParameters=true}
  ]
  rendererClasses [
   &NiagaraSpriteRendererProperties
   &NiagaraMeshRendererProperties
   &NiagaraLightRendererProperties
  ]
  renderers [
   {rendererClass=&NiagaraSpriteRendererProperties}
   {rendererClass=&NiagaraSpriteRendererProperties, rendererIndex=1}
   {rendererClass=&NiagaraSpriteRendererProperties, rendererIndex=2}
   {rendererClass=&NiagaraSpriteRendererProperties, rendererIndex=3}
   {rendererClass=&NiagaraSpriteRendererProperties, rendererIndex=4}
   {rendererClass=&NiagaraMeshRendererProperties, rendererIndex=5}
   {rendererClass=&NiagaraMeshRendererProperties, rendererIndex=6}
   {rendererClass=&NiagaraMeshRendererProperties, rendererIndex=7}
   {rendererClass=&NiagaraMeshRendererProperties, rendererIndex=8}
   {rendererClass=&NiagaraLightRendererProperties, rendererIndex=9}
   {rendererClass=&NiagaraSpriteRendererProperties, rendererIndex=10}
   {rendererClass=&NiagaraSpriteRendererProperties, rendererIndex=11}
   {rendererClass=&NiagaraLightRendererProperties, rendererIndex=12}
  ]
 }
 HandTrails {
  emitterUpdateScript [
   &EmitterState()
   &SpawnRate(SpawnRate=&Multiply_Float(A=60, B=$User.HandTrails), Use Spawn Probability=false)
  ]
  particleSpawnScript [
   &FN_SetSpawnGroup(Number of Spawn Groups=2)
   &InitializeParticle(
    Color=<rgba 1.5 1.5 1.5 1.5>
    Lifetime=0.325
    Ribbon Width=20
    Ribbon Width Mode=::NewEnumerator1
    Sprite Size Mode=::NewEnumerator3
    Uniform Sprite Size=28
    UsePositionOffset=false
    Write Material Random=true
   )
   &SkeletalMeshLocation(
    Bone / Socket Index=&SpawnGroup()
    Bone Sampling Mode=::NewEnumerator4
    MaskSpawnGroup=false
    Sampled Position Offset=&MergeHelper_VectorSwitch(0=<10 0 0>, 1=<-10 0 0>, 2=?, 3=?, Index=&SpawnGroup(), Num=::NewEnumerator0)
    Skeletal Mesh=$System.Skeletal Mesh Reader
   )
  ]
  particleUpdateScript [
   &ParticleState()
   &SolveForcesAndVelocity(Clamp Velocity=false, Limit Acceleration=false)
   &ScaleRibbonWidth(Ribbon Width Scale=&FloatFromCurve(FloatCurve=DataInterface() {propertyValues {Curve {
        defaultValue=1
        keys [{interpMode=RCIM_Cubic, tangentMode=RCTM_User, time=0, value=1}, {interpMode=RCIM_Cubic, tangentMode=RCTM_User, time=1, value=0}]
       }}}))
   &ScaleColor(
    Scale Alpha=&FloatFromCurve(
     FloatCurve=DataInterface() {propertyValues {Curve {keys [
         {arriveTangent=-4.0308, interpMode=RCIM_Cubic, leaveTangent=-4.0308, tangentMode=RCTM_User, time=0, value=1}
         {arriveTangent=-1, interpMode=RCIM_Cubic, leaveTangent=-1, time=0.174, value=0.3372}
         <at 1 0 RCIM_Cubic>
        ]}}}
     Scale Curve=&Multiply_Float(A=0.05, B=&FNTimeOfDayScalar(Night Scale=0.6, Morning Scale=1, Evening Scale=1))
    )
    Scale RGB=&MakeVectorFromLinearColorRGB(Color=&ColorFromCurve(ColorCurve=DataInterface() {propertyValues {
        AlphaCurve {keys []}
        BlueCurve {keys [<at 0.1793 0.55>,<at 0.3519 0.4449 RCIM_Cubic>]}
        GreenCurve {keys [<at 0.1793 0.3438>,<at 0.3519 0.178 RCIM_Cubic>]}
        RedCurve {keys [<at 0.1793 1>,<at 0.3519 1.2 RCIM_Cubic>]}
       }}))
    ScaleA=true
    ScaleRGB=true
   )
   &DynamicMaterialParameters(
    Index 0 Param 1=&FloatFromCurve(FloatCurve=DataInterface() {propertyValues {Curve {
        defaultValue=1
        keys [{interpMode=RCIM_Cubic, tangentMode=RCTM_User, time=0, value=1}, {interpMode=RCIM_Cubic, tangentMode=RCTM_User, time=1, value=0}]
       }}})
    Index 0 Param 2=3
    Index 0 Param 3=1
    Index 0 Param 4=1
    Module.Index0Float0Write=true
    Module.Index0Float1Write=true
    Module.Index0Float2Write=true
    Module.Index0Float3Write=true
   )
   /SetVariables(Particles.GlowColor=&MakeLinearColorFromVectorAndFloat(
     "Vector (RGB)"=&MakeVectorFromLinearColorRGB(Color=<rgba 1 1 1 1>)
     "Float (Alpha)"=&FloatFromCurve(FloatCurve=DataInterface() {propertyValues {Curve {
         defaultValue=1
         keys [
          {arriveTangent=9.0725, interpMode=RCIM_Cubic, leaveTangent=9.0725, tangentMode=RCTM_User, time=0, value=0}
          {arriveTangent=-2.5249, interpMode=RCIM_Cubic, leaveTangent=-2.5249, tangentMode=RCTM_User, time=0.1408, value=0.9847}
          {arriveTangent=-0.0405, interpMode=RCIM_Cubic, leaveTangent=-0.0405, tangentMode=RCTM_User, time=1, value=0}
         ]
        }}})
    )) {isSetParameters=true}
   &BlendSpace("Position Blend (World <-> Local)":NiagaraFloat=&MakeCustomFloatFromBool(
     Bool=&SetBoolByFloatComparison(
      A=&VectorLength(Vector=&VectorFromFloat(Value=$User.OwnerLocalVelocity))
      B=600
     )
     True Float=0
     False Float=1
     Translation Contribution:NiagaraFloat=1
     Rotation Contribution:NiagaraFloat=1
    ))
  ]
  rendererClasses [&NiagaraRibbonRendererProperties, &NiagaraSpriteRendererProperties]
  renderers [
   {rendererClass=&NiagaraRibbonRendererProperties}
   {rendererClass=&NiagaraSpriteRendererProperties, rendererIndex=1}
  ]
 }
 TrailSparkles {
  emitterUpdateScript [
   &EmitterState()
   &SpawnRate(SpawnRate=&Multiply_Float(A=120, B=$User.HandTrails), Use Spawn Probability=false)
  ]
  particleSpawnScript [
   &FN_SetSpawnGroup(Number of Spawn Groups=2)
   &InitializeParticle(
    Color=<rgba 3 0.4911 2.535 1>
    Lifetime Max=0.4
    Lifetime Min=0.1
    Lifetime Mode=::NewEnumerator1
    Sprite Size Mode=::NewEnumerator3
    Uniform Sprite Size=&RandomRangeFloat(Minimum=0.8, Maximum=1.8)
    UsePositionOffset=false
    Write Material Random=true
   )
   &SkeletalMeshLocation(
    Bone / Socket Index=&SpawnGroup()
    Bone Sampling Mode=::NewEnumerator4
    MaskSpawnGroup=false
    Sampled Position Offset=&MergeHelper_VectorSwitch(0=<10 0 0>, 1=<-10 0 0>, 2=?, 3=?, Index=&SpawnGroup(), Num=::NewEnumerator0)
    Skeletal Mesh=$System.Skeletal Mesh Reader
   )
   &ShapeLocation(
    MaskSpawnGroup=false
    Override Seed=false
    Shape Origin=$Particles.SkeletalMeshLocation.SampledPosition
    Sphere Radius=12
    UseOwnerScale=true
   )
   &AddVelocity(
    Cone Angle=60
    Cone Axis=&ReturnMeshOrientationAxis(Mesh Orientation Axis=::YAxis, Mesh Orientation Quat=$Particles.MeshOrientation)
    Module.UseConeFalloff=false
    Module.UseVelDistribution=false
    Override Seed=false
    Velocity Mode=::NewEnumerator2
    Velocity Speed=&RandomRangeFloat(Minimum=40, Maximum=120)
   )
  ]
  particleUpdateScript [
   &ParticleState()
   /SetVariables(
    Particles.GlowColor=&MultiplyLinearColor(Linear Color=$Particles.Color, Scale Factor=&MakeVector4(X=0.75, Y=0.75, Z=0.75, W=0.05))
    Particles.GlowSize=&Multiply_Vector2D_ByFloat(Float=5, Vector2D=$Particles.SpriteSize)
   ) {isSetParameters=true}
   &Drag(Drag=&RandomRangeFloat(Minimum=1, Maximum=3), Use Linear Drag=true, Use Rotational Drag=true)
   &SolveForcesAndVelocity(Clamp Velocity=false, Limit Acceleration=false)
   &ScaleColor(
    Scale Alpha=&FloatFromCurve(FloatCurve=DataInterface() {propertyValues {Curve {
        defaultValue=1
        keys [
         {arriveTangent=-2.0974, interpMode=RCIM_Cubic, leaveTangent=-2.0974, tangentMode=RCTM_User, time=0, value=1}
         {interpMode=RCIM_Cubic, tangentMode=RCTM_User, time=1, value=0}
        ]
       }}})
    Scale RGB=&VectorFromFloat(Value=&RandomRangeFloat(Minimum=8, Maximum=15))
    ScaleA=true
    ScaleRGB=true
   )
   &BlendSpace("Position Blend (World <-> Local)":NiagaraFloat=&MakeCustomFloatFromBool(
     Bool=&SetBoolByFloatComparison(
      A=&VectorLength(Vector=&VectorFromFloat(Value=$User.OwnerLocalVelocity))
      B=600
     )
     True Float=0
     False Float=1
     Translation Contribution:NiagaraFloat=1
     Rotation Contribution:NiagaraFloat=1
    ))
  ]
  rendererClasses [&NiagaraSpriteRendererProperties]
  renderers [
   {rendererClass=&NiagaraSpriteRendererProperties}
   {rendererClass=&NiagaraSpriteRendererProperties, rendererIndex=1}
  ]
 }
}
summary {
 systemName=NS_Emote_PearlTaffyOpera
 emitters [
  {
   emitterName=Environment
   rendererClasses [
    &NiagaraSpriteRendererProperties
    &NiagaraMeshRendererProperties
    &NiagaraLightRendererProperties
   ]
  }
  {emitterName=HandTrails, rendererClasses [&NiagaraRibbonRendererProperties, &NiagaraSpriteRendererProperties]}
  {emitterName=TrailSparkles, rendererClasses [&NiagaraSpriteRendererProperties]}
 ]
 userVariables [
  User.AirflowDustSpawnRate {type=NiagaraFloat, defaultValue=0, description=""}
  User.bIsFrontEnd {type=NiagaraBool}
  User.bIsFrontEndPreview {type=NiagaraBool}
  User.BurstIndex {type=NiagaraFloat, defaultValue=0, description=""}
  User.HandTrails {type=NiagaraFloat, defaultValue=0, description=""}
  User.OwnerLocalVelocity {type=NiagaraFloat, defaultValue=0, description=""}
  User.ScaleLifetime {type=NiagaraFloat, defaultValue=2.5, description=""}
  User.Variant {type=NiagaraInt32}
 ]
}
