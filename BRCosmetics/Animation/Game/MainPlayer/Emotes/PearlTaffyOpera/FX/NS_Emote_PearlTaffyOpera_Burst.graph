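# Niagara burst effect for the PearlTaffyOpera emote: a brief pink and purple pop with hearts, rays, rainbow, sparkles and glows.
# SimpleSpriteBurst uses spawn groups and visibility tags to share one emitter across mesh, sprite and light renderers.
# The airflow emitter adds two spinning, growing layered meshes.
# AirflowDust throws curl-noised particles from two tilted, age-widening shapes, with extra continuous spawning scaled by User.AirflowDustSpawnRate (default zero).
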NiagaraSystem /BRCosmetics/Animation/Game/MainPlayer/Emotes/PearlTaffyOpera/FX/NS_Emote_PearlTaffyOpera_Burst
dependencies {
 usedDataInterfaces [
  &NiagaraDataInterfaceCurve
  &NiagaraDataInterfaceVectorCurve
  &NiagaraDataInterfaceVector2DCurve
  &NiagaraDataInterfaceArrayColor
 ]
 usedDynamicInputs []
 usedModules [&SystemState]
}
emitters {
 AirflowDust {
  emitterUpdateScript [
   &EmitterState()
   &SpawnBurst_Instantaneous(Spawn Count=12, Use Spawn Probability=false, UseLoopCountLimit=false)
   &SpawnRate(SpawnRate=&Multiply_Float(A=$User.AirflowDustSpawnRate, B=400), Use Spawn Probability=false)
   &SpawnBurst_Instantaneous(Spawn Count=8, Use Spawn Probability=false, UseLoopCountLimit=false) {module=SpawnBurst_Instantaneous001}
   &SpawnRate(SpawnGroup=1, SpawnRate=&Multiply_Float(A=$User.AirflowDustSpawnRate, B=96), Use Spawn Probability=false) {module=SpawnRate001}
  ]
  particleSpawnScript [
   &PartitionParticles(Particles Per Partition=2, Partition Mode:ENiagara_PartitionMode_v2=::NewEnumerator1)
   /SetVariables(Particles.VisibilityTag=&SpawnGroup()) {isSetParameters=true}
   &InitializeParticle(
    Color=&MaskLinearColorBySpawnGroup(
     Masked Linear Color=&SelectLinearColorFromArray(
      Array Sampling Mode=::NewEnumerator0
      Direct Array Index=?
      Color Selection Array=DataInterface() {propertyValues {ColorData [
         <rgba 1 0.0714 0.5244 1>
         <rgba 0.8187 0.1885 1 1>
         <rgba 0.8673 0.4062 1 1>
         <rgba 1 0.2928 0.8097 1>
         <rgba 1 0.4186 0.8141 1>
         <rgba 0.5365 0.0589 0.4715 1>
         <rgba 0.3147 0.0268 0.4062 1>
         <rgba 1 0.2448 0.5577 1>
        ]}}
      Random Seed=0
     )
     Passthrough Linear Color=<rgba 2 1.1458 1.1458 1>
    )
    Lifetime=&MaskFloatBySpawnGroup(
     Masked Float=&RandomRangeFloat(Minimum=0.3, Maximum=0.6)
     Passthrough Float=&RandomRangeFloat(Minimum=0.3, Maximum=0.8)
    )
    Position Offset=&MaskVector(
     Mask Source=1
     Mask Parameter=$Particles.PartitionParticles.Partition
     Mask=0
     Masked Vector=<-22.5 -9.7068 92>
     Passthrough Vector=<-22.5 0 80>
    )
    Sprite Size Mode=::NewEnumerator3
    Uniform Sprite Size=&MaskFloatBySpawnGroup(
     Masked Float=&UniformAOrBFloat(A=1.2, B=3, Distribution Weight=0.25, Evaluation Type=::NewEnumerator0, Recalculate Random Each Loop=true, Override Seed=false)
     Passthrough Float=&RandomRangeFloat(Minimum=10, Maximum=25)
    )
    UsePositionOffset=true
    Write Material Random=true
   )
   &ShapeLocation(
    Handle Radius=5.0063
    Large Radius=&FloatFromCurve(
     FloatCurve=DataInterface() {propertyValues {Curve {keys [{arriveTangent=718.4051, interpMode=RCIM_Cubic, leaveTangent=718.4051, tangentMode=RCTM_User, time=0, value=25}, <at 0.275 105 RCIM_Cubic>]}}}
     CurveIndex=$Emitter.Age
    )
    MaskSpawnGroup=false
    Override Seed=false
    Rotation Mode=::NewEnumerator2
    Shape Primitive=::NewEnumerator3
    Surface Distribution=1
    U Distribution:NiagaraFloat=&FloatFromCurve(FloatCurve=DataInterface() {propertyValues {Curve {keys [<at 0 -0.25 RCIM_Cubic>,<at 0.12 0.35 RCIM_Cubic>]}}}, CurveIndex=$Emitter.Age)
    UseOwnerScale=true
    Yaw / Pitch / Roll=&MaskVector(
     Mask Source=1
     Mask Parameter=$Particles.PartitionParticles.Partition
     Mask=0
     Masked Vector=<-0.088 0.215 0.472>
     Passthrough Vector=<-0.7591 -0.1238 0.3301>
    )
   )
   &AddVelocity(
    Cone Axis=$Particles.ShapeLocation.ShapeBitangent
    Module.UseConeFalloff=false
    Module.UseVelDistribution=false
    Override Seed=false
    Velocity Mode=::NewEnumerator2
    Velocity Speed=&MaskFloatBySpawnGroup(
     Masked Float=&RandomRangeFloat(Minimum=225, Maximum=410)
     Passthrough Float=&RandomRangeFloat(Minimum=120, Maximum=300)
    )
   )
  ]
  particleUpdateScript [
   &ParticleState()
   &SpriteRotationRate(Rotation Rate=&RandomRangeFloat(Minimum=-500, Maximum=500))
   &ScaleSpriteSize(
    Scale Sprite Size Mode=::NewEnumerator0
    Uniform Scale Factor=&MaskFloatBySpawnGroup(
     Masked Float=1
     Passthrough Float=&FloatFromCurve(FloatCurve=DataInterface() {propertyValues {Curve {
         defaultValue=1
         keys [
          {arriveTangent=9.0725, interpMode=RCIM_Cubic, leaveTangent=9.0725, tangentMode=RCTM_User, time=0, value=0}
          {arriveTangent=-0.0031, interpMode=RCIM_Cubic, leaveTangent=-0.0031, tangentMode=RCTM_User, time=0.1408, value=0.9847}
          {arriveTangent=-2.0324, interpMode=RCIM_Cubic, leaveTangent=-2.0324, tangentMode=RCTM_User, time=1, value=0}
         ]
        }}})
    )
   )
   &Drag(
    Drag=&MaskFloatBySpawnGroup(
     Masked Float=&RandomRangeFloat(Minimum=1, Maximum=3)
     Passthrough Float=&RandomRangeFloat(Minimum=0.8, Maximum=1.2)
    )
    Use Linear Drag=true
    Use Rotational Drag=true
   )
   &CurlNoiseForce(Bias Noise=false, Noise Strength=&RandomRangeFloat(Minimum=120, Maximum=325), Pan Noise=false)
   &SolveForcesAndVelocity(Clamp Velocity=false, Limit Acceleration=false)
   &ScaleColor(
    Scale Alpha=&MaskFloatBySpawnGroup(
     Masked Float=&FloatFromCurve(FloatCurve=DataInterface() {propertyValues {Curve {
         defaultValue=1
         keys [
          {arriveTangent=9.0725, interpMode=RCIM_Cubic, leaveTangent=9.0725, tangentMode=RCTM_User, time=0, value=0}
          {arriveTangent=0.0337, interpMode=RCIM_Cubic, leaveTangent=0.0337, tangentMode=RCTM_User, time=0.1408, value=0.9847}
          {arriveTangent=-3.0118, interpMode=RCIM_Cubic, leaveTangent=-3.0118, tangentMode=RCTM_User, time=1, value=0}
         ]
        }}})
     Passthrough Float=&FloatFromCurve(FloatCurve=DataInterface() {propertyValues {Curve {
         defaultValue=1
         keys [
          {arriveTangent=9.0725, interpMode=RCIM_Cubic, leaveTangent=9.0725, tangentMode=RCTM_User, time=0, value=0}
          {arriveTangent=-2.5249, interpMode=RCIM_Cubic, leaveTangent=-2.5249, tangentMode=RCTM_User, time=0.1408, value=0.9847}
          {arriveTangent=-0.0405, interpMode=RCIM_Cubic, leaveTangent=-0.0405, tangentMode=RCTM_User, time=1, value=0}
         ]
        }}})
    )
    Scale RGB=&MaskVectorBySpawnGroup(
     Masked Vector=&VectorFromFloat(Value=&Waveform("Waveform [1]"=::NewEnumerator0, "[1] Frequency"=1, "[1] Amplitude Min/Max"=<18 32>))
     Passthrough Vector=<1 1 1>
    )
    ScaleA=true
    ScaleRGB=true
   )
   /SetVariables(
    Particles.GlowColor=&MultiplyLinearColor(Linear Color=$Particles.Color, Scale Factor=&MakeVector4(X=0.75, Y=0.75, Z=0.75, W=0.05))
    Particles.GlowSize=&Multiply_Vector2D_ByFloat(Float=5, Vector2D=$Particles.SpriteSize)
   ) {isSetParameters=true}
   /SetVariables(
    Particles.HeartGlowColor:LinearColor=&MultiplyLinearColor(
     Linear Color=<rgba 0.5 0.5 0.5 1>
     Scale Factor=&MakeVector4FromVectorAndFloat(
      "Vector (RGB)"=&MakeVectorFromLinearColorRGB(Color=&RandomRangeLinearColor(Minimum=<rgba 0.9639 1.7703 5 1>, Maximum=<rgba 5 1.2208 2.3366 1>))
      "Float (Alpha)"=&RandomRangeFloat(Minimum=0.1, Maximum=0.225)
     )
    )
    Particles.HeartGlowSize:Vector2f=&Multiply_Vector2D_ByFloat(Float=2, Vector2D=$Particles.SpriteSize)
   ) {isSetParameters=true}
   /SetVariables(
    Particles.SparkleColor:LinearColor=&MakeLinearColorFromVectorAndFloat(
     "Vector (RGB)"=&MakeVectorFromLinearColorRGB(Color=&RandomRangeLinearColor(Minimum=<rgba 0.93 0.15 1 1>, Maximum=<rgba 1 0.2548 0.7529 1>))
     "Float (Alpha)"=&FloatFromCurve(
      FloatCurve=DataInterface() {propertyValues {Curve {
         defaultValue=1
         keys [
          {arriveTangent=9.0725, interpMode=RCIM_Cubic, leaveTangent=9.0725, tangentMode=RCTM_User, time=0, value=0}
          {arriveTangent=-0.2024, interpMode=RCIM_Cubic, leaveTangent=-0.2024, tangentMode=RCTM_User, time=0.1408, value=0.9847}
          {arriveTangent=-1.809, interpMode=RCIM_Cubic, leaveTangent=-1.809, tangentMode=RCTM_User, time=1, value=0}
         ]
        }}}
      Scale Curve=&Waveform("Waveform [1]"=::NewEnumerator2, "[1] Frequency"=5, "[1] Amplitude Min/Max"=<0.3 1>)
     )
    )
    Particles.SparkleSize:Vector2f=&Vector2DFromFloat(Value=&FloatFromCurve(
      FloatCurve=DataInterface() {propertyValues {Curve {
         defaultValue=1
         keys [
          {arriveTangent=9.0725, interpMode=RCIM_Cubic, leaveTangent=9.0725, tangentMode=RCTM_User, time=0, value=0}
          {arriveTangent=-0.0031, interpMode=RCIM_Cubic, leaveTangent=-0.0031, tangentMode=RCTM_User, time=0.1408, value=0.9847}
          {arriveTangent=-2.0324, interpMode=RCIM_Cubic, leaveTangent=-2.0324, tangentMode=RCTM_User, time=1, value=0}
         ]
        }}}
      Scale Curve=&RandomRangeFloat(Minimum=50, Maximum=85)
     ))
   ) {isSetParameters=true}
  ]
  rendererClasses [&NiagaraSpriteRendererProperties]
  renderers [
   {rendererClass=&NiagaraSpriteRendererProperties}
   {rendererClass=&NiagaraSpriteRendererProperties, rendererIndex=1}
   {rendererClass=&NiagaraSpriteRendererProperties, rendererIndex=2}
   {rendererClass=&NiagaraSpriteRendererProperties, rendererIndex=3}
   {rendererClass=&NiagaraSpriteRendererProperties, rendererIndex=4}
  ]
 }
 SimpleSpriteBurst {
  emitterUpdateScript [
   &EmitterState()
   &SpawnBurst_Instantaneous(Use Spawn Probability=false, UseLoopCountLimit=false) {module=SpawnBurst_Instantaneous001}
   &SpawnBurst_Instantaneous(Spawn Count=16, Spawn Group=1, Use Spawn Probability=false, UseLoopCountLimit=false)
   &SpawnBurst_Instantaneous(
    Spawn Count=&RandomRangeInt(Minimum=84, Maximum=128)
    Spawn Group=2
    Use Spawn Probability=false
    UseLoopCountLimit=false
   ) {module=SpawnBurst_Instantaneous002}
   &SpawnBurst_Instantaneous(Spawn Group=3, Spawn Time=0.08, Use Spawn Probability=false, UseLoopCountLimit=false) {module=SpawnBurst_Instantaneous003}
   &SpawnBurst_Instantaneous(Spawn Group=4, Spawn Time=0.04, Use Spawn Probability=false, UseLoopCountLimit=false) {enabled=false, module=SpawnBurst_Instantaneous004}
   &SpawnRate(
    SpawnGroup=4
    SpawnRate=&FloatFromCurve(
     FloatCurve=DataInterface() {propertyValues {Curve {keys [<at 0 1 RCIM_Constant>,<at 0.0125 0 RCIM_Constant>]}}}
     CurveIndex=$Emitter.NormalizedLoopAge
     Scale Curve=24
    )
    Use Spawn Probability=false
   )
   /SetVariables(Emitter.EdgeFade:NiagaraFloat=&FloatFromCurve(FloatCurve=DataInterface() {propertyValues {Curve {keys [<at 0 3>,<at 0.5 1>]}}}, CurveIndex=$Emitter.NormalizedLoopAge)) {isSetParameters=true}
  ]
  particleSpawnScript [
   /SetVariables(Particles.VisibilityTag=&SpawnGroup()) {isSetParameters=true}
   &InitializeParticle(
    Color=<rgba 7 1.5056 5.6737 1>
    Lifetime=&MergeHelper_FloatSwitch(
     0=0.6
     1=&RandomRangeFloat(Minimum=0.6, Maximum=0.85)
     2=&RandomRangeFloat(Minimum=0.3, Maximum=0.9)
     3=0.3
     4=0.6
     Index=&SpawnGroup()
     Num=::NewEnumerator3
    )
    Mesh Scale=&Multiply_VectorByFloat(Vector=<15 10 10>, Float=0.8)
    Mesh Scale Mode=::NewEnumerator4
    Position Offset=&MergeHelper_VectorSwitch(
     0=<-25 -35 75>
     1=<-25 -40 105>
     2=<-25 4 105>
     3=<-25 4 105>
     4=<-25 -35 90>
     Index=&SpawnGroup()
     Num=::NewEnumerator3
    )
    Sprite Rotation Angle=&MergeHelper_FloatSwitch(
     0=0
     1=&RandomRangeFloat(Minimum=0, Maximum=360)
     2=0
     3=0
     Index=&SpawnGroup()
     Num=::NewEnumerator2
    )
    Sprite Rotation Mode=::NewEnumerator0
    Sprite Size Mode=::NewEnumerator3
    Uniform Sprite Size=350
    UsePositionOffset=true
    Write Material Random=true
   )
   /SetVariables(Particles.Rayorientation:Quat4f=&MakeQuaternion(
     Quaternion From=5
     Axis=?
     Angle=&RandomRangeFloat(Minimum=0, Maximum=1)
     Yaw=?
     Pitch=?
     Roll=?
     X=?
     Y=?
     Z=?
     From=?
     To=?
     Stabilization Mode:ENiagara_OrientVectorCount=::NewEnumerator0
     Side Direction:Vector3f=?
     Up Direction:Vector3f=?
     Angle Type=::NewEnumerator0
     Coordinate Space=?
     Normalize:NiagaraBool=true
    )) {isSetParameters=true}
   &AddVelocity(
    Module.Use Falloff Distance=false
    Origin Offset=<-25 4 105>
    Override Seed=false
    Velocity Mode=::NewEnumerator1
    Velocity Speed=&MergeHelper_FloatSwitch(
     0=0
     1=0
     2=&RandomRangeFloat(Minimum=325, Maximum=500)
     3=?
     Index=&SpawnGroup()
     Num=::NewEnumerator1
    )
   )
  ]
  particleUpdateScript [
   &ParticleState()
   &ScaleSpriteSize(Uniform Curve Sprite Scale=DataInterface() {propertyValues {Curve {
       defaultValue=0
       keys [
        {interpMode=RCIM_Cubic, tangentMode=RCTM_User, time=0, value=0.4}
        {interpMode=RCIM_Cubic, tangentMode=RCTM_User, time=0.5, value=1}
        {tangentMode=RCTM_User, time=1, value=0}
       ]
      }}})
   &ScaleMeshSize(Scale Factor=&Multiply_VectorByFloat(
     Vector=&VectorFromCurve(VectorCurve=DataInterface() {propertyValues {
        XCurve {keys [<at 0 1>,<at 1 1>]}
        YCurve {keys [<at 0 1>,<at 1 1>]}
        ZCurve {
         defaultValue=1
         keys [
          <at 0 1 RCIM_Cubic>
          {arriveTangent=-0.1459, interpMode=RCIM_Cubic, leaveTangent=-0.1459, time=0.1379, value=1.0851}
          {arriveTangent=-0.0815, interpMode=RCIM_Cubic, leaveTangent=-0.0815, time=0.3339, value=0.9513}
          {arriveTangent=-0.0595, interpMode=RCIM_Cubic, leaveTangent=-0.0595, time=0.4821, value=1.0571}
          {arriveTangent=-0.1102, interpMode=RCIM_Cubic, leaveTangent=-0.1102, time=0.7313, value=0.9276}
          <at 1 1 RCIM_Cubic>
         ]
        }
       }})
     Float=&MakeFloatFromVector(
      Vector=&VectorFromFloat(Value=&FloatFromCurve(FloatCurve=DataInterface() {propertyValues {Curve {
           defaultValue=1
           keys [
            {arriveTangent=2.3452, interpMode=RCIM_Cubic, leaveTangent=2.3452, tangentMode=RCTM_User, time=0, value=0.3}
            {arriveTangent=1.1179, interpMode=RCIM_Cubic, leaveTangent=1.1179, tangentMode=RCTM_User, time=0.2042, value=0.7}
            {arriveTangent=0.0224, interpMode=RCIM_Cubic, leaveTangent=0.0224, tangentMode=RCTM_User, time=1, value=1}
           ]
          }}}))
      Channel=::NewEnumerator0
     )
    )) {module=ScaleMeshSize001}
   &DynamicMaterialParameters(
    Index 0 Param 1=&FloatFromCurve(
     FloatCurve=DataInterface() {propertyValues {Curve {
        defaultValue=1
        keys [
         {arriveTangent=9.0725, interpMode=RCIM_Cubic, leaveTangent=9.0725, tangentMode=RCTM_User, time=0, value=0}
         {arriveTangent=-2.5249, interpMode=RCIM_Cubic, leaveTangent=-2.5249, tangentMode=RCTM_User, time=0.1408, value=0.9847}
         {arriveTangent=-0.0405, interpMode=RCIM_Cubic, leaveTangent=-0.0405, tangentMode=RCTM_User, time=1, value=0}
        ]
       }}}
     Scale Curve=3
    )
    Index 0 Param 3=0.5473
    Module.Index0Float0Write=true
    Module.Index0Float1Write=false
    Module.Index0Float2Write=false
    Module.Index0Float3Write=false
   )
   &ScaleColor(
    Scale Alpha=&FloatFromCurve(
     FloatCurve=DataInterface() {propertyValues {Curve {
        defaultValue=0
        keys [
         {arriveTangent=9.0725, interpMode=RCIM_Cubic, leaveTangent=9.0725, tangentMode=RCTM_User, time=0, value=0}
         {arriveTangent=-2.5249, interpMode=RCIM_Cubic, leaveTangent=-2.5249, tangentMode=RCTM_User, time=0.1408, value=0.9847}
         {arriveTangent=-0.0405, interpMode=RCIM_Cubic, leaveTangent=-0.0405, tangentMode=RCTM_User, time=1, value=0}
        ]
       }}}
     Scale Curve=0.0075
    )
    Scale RGB=&VectorFromFloat(Value=&FloatFromCurve(
      FloatCurve=DataInterface() {propertyValues {Curve {keys [<at 0 1>, <at 0.1234 2>, {arriveTangent=-1.1407, interpMode=RCIM_Cubic, leaveTangent=-1.1407, time=0.3119, value=1.5}, <at 1 1>]}}}
      Scale Curve=0.7
     ))
    ScaleA=true
    ScaleRGB=true
   )
   &Drag(Drag=&RandomRangeFloat(Minimum=1, Maximum=3), Use Linear Drag=true, Use Rotational Drag=true)
   &CurlNoiseForce(
    Bias Noise=false
    Noise Strength=&MergeHelper_FloatSwitch(
     0=0
     1=0
     2=&RandomRangeFloat(Minimum=120, Maximum=325)
     3=?
     Index=&SpawnGroup()
     Num=::NewEnumerator1
    )
    Pan Noise=false
   )
   &SolveForcesAndVelocity(Clamp Velocity=false, Limit Acceleration=false)
   /SetVariables(
    Particles.GlowCO=-20
    Particles.SpriteGlowColor:LinearColor=&MultiplyLinearColor(
     Linear Color=$Particles.Color
     Scale Factor=&MakeVector4FromVectorAndFloat("Vector (RGB)"=&VectorFromFloat(Value=16), "Float (Alpha)"=0.4)
    )
    Particles.SpritePos:NiagaraPosition=&AddVectorToPosition(Position=$Engine.Emitter.SimulationPosition, Vector=<-15 0 95>)
   ) {isSetParameters=true}
   &FN_Light_Attributes(
    Color=<rgba 1 0.3214 0.7112 10>
    Create Color Attribute=true
    Create Exponent Attribute=true
    Create Position Attribute=true
    Create Radius Attribute=true
    Enable Particle Light Enable=true
    Exponent=5
    Light Enabled=true
    Position:NiagaraPosition=$Particles.Position
    Radius=&FloatFromCurve(
     FloatCurve=DataInterface() {propertyValues {Curve {
        defaultValue=0
        keys [
         {arriveTangent=9.0725, interpMode=RCIM_Cubic, leaveTangent=9.0725, tangentMode=RCTM_User, time=0, value=0}
         {arriveTangent=0.2697, interpMode=RCIM_Constant, tangentMode=RCTM_User, time=0.3, value=0.9847}
         {arriveTangent=-1.4067, interpMode=RCIM_Cubic, leaveTangent=-1.4067, time=0.7, value=1}
         {arriveTangent=-0.0405, interpMode=RCIM_Cubic, leaveTangent=-0.0405, tangentMode=RCTM_User, time=1, value=0}
        ]
       }}}
     Scale Curve=75
    )
    Scale Color=&FloatFromCurve(FloatCurve=DataInterface() {propertyValues {Curve {
        defaultValue=0
        keys [
         {arriveTangent=9.0725, interpMode=RCIM_Cubic, leaveTangent=9.0725, tangentMode=RCTM_User, time=0, value=0}
         {arriveTangent=0.2697, interpMode=RCIM_Constant, tangentMode=RCTM_User, time=0.3, value=0.9847}
         {arriveTangent=-1.4067, interpMode=RCIM_Cubic, leaveTangent=-1.4067, time=0.4898, value=1}
         {arriveTangent=-0.0405, interpMode=RCIM_Cubic, leaveTangent=-0.0405, tangentMode=RCTM_User, time=1, value=0}
        ]
       }}})
   )
   /SetVariables(
    Particles.RayColor:LinearColor=&MakeLinearColorFromVectorAndFloat(
     "Vector (RGB)"=&Multiply_VectorByFloat(
      Vector=&MakeVectorFromLinearColorRGB(Color=&RandomRangeLinearColor(Minimum=<rgba 15.4069 8.2034 25 1>, Maximum=<rgba 25 9.8842 14.5733 1>))
      Float=&FloatFromCurve(
       FloatCurve=DataInterface() {propertyValues {Curve {
          defaultValue=1
          keys [
           {arriveTangent=9.0725, interpMode=RCIM_Cubic, leaveTangent=9.0725, tangentMode=RCTM_User, time=0, value=0}
           {arriveTangent=-2.5249, interpMode=RCIM_Cubic, leaveTangent=-2.5249, tangentMode=RCTM_User, time=0.1408, value=0.9847}
           {arriveTangent=-0.0405, interpMode=RCIM_Cubic, leaveTangent=-0.0405, tangentMode=RCTM_User, time=1, value=0}
          ]
         }}}
       Scale Curve=&RandomRangeFloat(Minimum=0.125, Maximum=0.2)
      )
     )
     "Float (Alpha)"=&FloatFromCurve(
      FloatCurve=DataInterface() {propertyValues {Curve {
         defaultValue=1
         keys [
          {arriveTangent=9.0725, interpMode=RCIM_Cubic, leaveTangent=9.0725, tangentMode=RCTM_User, time=0, value=0}
          {arriveTangent=-2.5249, interpMode=RCIM_Cubic, leaveTangent=-2.5249, tangentMode=RCTM_User, time=0.1408, value=0.9847}
          {arriveTangent=-0.0405, interpMode=RCIM_Cubic, leaveTangent=-0.0405, tangentMode=RCTM_User, time=1, value=0}
         ]
        }}}
      Scale Curve=&RandomRangeFloat(Minimum=0.125, Maximum=0.3)
     )
    )
    Particles.RaySize:Vector2f=&Vector2DFromCurve(
     Vector2Curve=DataInterface() {propertyValues {
       XCurve {keys [{tangentMode=RCTM_User, time=0, value=1}, {tangentMode=RCTM_User, time=1, value=0.5}]}
       YCurve {keys [
         {arriveTangent=15.8763, interpMode=RCIM_Cubic, leaveTangent=15.8763, tangentMode=RCTM_User, time=0, value=0.5}
         {
          arriveTangent=-0.3643
          arriveTangentWeight=0.1414
          interpMode=RCIM_Cubic
          leaveTangent=-0.3643
          leaveTangentWeight=0.33
          tangentMode=RCTM_User
          time=0.4
          value=1
         }
         {tangentMode=RCTM_User, time=1, value=0.9}
        ]}
      }}
     ScaleCurve=&RandomRangeVector2D(Minimum=<60 80>, Maximum=<75 100>)
    )
   ) {isSetParameters=true}
   &SpriteRotationRate(Rotation Rate=&MergeHelper_FloatSwitch(
     0=0
     1=&RandomRangeFloat(Minimum=30, Maximum=75)
     2=0
     3=0
     Index=&SpawnGroup()
     Num=::NewEnumerator2
    ))
   &CameraOffset(Camera Offset Amount=&MergeHelper_FloatSwitch(0=20, 1=-30, 2=0, 3=0, 4=0, Index=&SpawnGroup(), Num=::NewEnumerator3))
   /SetVariables(
    Particles.SparkleColor:LinearColor=&MakeLinearColorFromVectorAndFloat(
     "Vector (RGB)"=&MakeVectorFromLinearColorRGB(Color=&SelectLinearColorFromArray(
       Array Sampling Mode=::NewEnumerator0
       Direct Array Index=?
       Color Selection Array=DataInterface() {propertyValues {ColorData [<rgba 8 3.1174 7.5938 1>, <rgba 5.5574 1.1209 8 1>, <rgba 8 1.865 3.0963 1>, <rgba 4.47 3.4176 8 1>, <rgba 8 0.9819 3.8157 1>]}}
       Random Seed=0
      ))
     "Float (Alpha)"=&FloatFromCurve(
      FloatCurve=DataInterface() {propertyValues {Curve {
         defaultValue=1
         keys [{interpMode=RCIM_Cubic, tangentMode=RCTM_User, time=0, value=1}, {interpMode=RCIM_Cubic, tangentMode=RCTM_User, time=1, value=0}]
        }}}
      Scale Curve=&Waveform(
       "Waveform [1]"=::NewEnumerator0
       "[1] Frequency"=&FloatFromCurve(
        FloatCurve=DataInterface() {propertyValues {Curve {keys [<at 0 1>,<at 0.7 1>,<at 1 3>]}}}
        Scale Curve=&RandomRangeFloat(Minimum=20, Maximum=35)
       )
       "[1] Amplitude Min/Max"=<0 1>
      )
     )
    )
    Particles.SparkleSize:Vector2f=&Vector2DFromFloat(Value=&FloatFromCurve(
      FloatCurve=DataInterface() {propertyValues {Curve {
         defaultValue=1
         keys [{interpMode=RCIM_Cubic, tangentMode=RCTM_User, time=0, value=1}, {interpMode=RCIM_Cubic, tangentMode=RCTM_User, time=1, value=0}]
        }}}
      Scale Curve=&RandomRangeFloat(Minimum=15, Maximum=25)
     ))
   ) {isSetParameters=true}
   /SetVariables(
    Particles.Glow Color=&MultiplyLinearColor(
     Linear Color=$Particles.SparkleColor
     Scale Factor=&MakeVector4FromVectorAndFloat("Vector (RGB)"=<1 1 1>, "Float (Alpha)"=0.025)
    )
    Particles.Glow Size=&Multiply_Vector2D_ByFloat(Float=1.25, Vector2D=$Particles.SparkleSize)
   ) {isSetParameters=true}
   /SetVariables(
    Particles.RainbowCO=-15
    Particles.RainbowColor:LinearColor=&MakeLinearColorFromVectorAndFloat(
     "Vector (RGB)"=&MakeVectorFromLinearColorRGB(Color=<rgba 1.45 0.3911 0.4188 1>)
     "Float (Alpha)"=&FloatFromCurve(
      FloatCurve=DataInterface() {propertyValues {Curve {
         defaultValue=1
         keys [
          <at 0 0 RCIM_Cubic>
          {arriveTangent=1.4268, interpMode=RCIM_Cubic, leaveTangent=1.4268, time=0.2, value=1}
          {arriveTangent=-0.0405, interpMode=RCIM_Cubic, leaveTangent=-0.0405, tangentMode=RCTM_User, time=0.7009, value=1}
          {arriveTangent=-8.0089, interpMode=RCIM_Cubic, leaveTangent=-8.0089, tangentMode=RCTM_User, time=1, value=0}
         ]
        }}}
      Scale Curve=1.45
     )
    )
    Particles.RainbowSize:Vector2f=&Multiply_Vector2D_ByFloat(
     Float=&FloatFromCurve(FloatCurve=DataInterface() {propertyValues {Curve {
         defaultValue=1
         keys [
          {arriveTangent=1.1179, interpMode=RCIM_Cubic, leaveTangent=1.1179, tangentMode=RCTM_User, time=0, value=0.5}
          {arriveTangent=0.0224, interpMode=RCIM_Cubic, leaveTangent=0.0224, tangentMode=RCTM_User, time=1, value=1}
         ]
        }}})
     Vector2D=<300 200>
    )
   ) {isSetParameters=true}
   /SetVariables(Particles.HeartFresnelColor:LinearColor=&MakeLinearColorFromVectorAndFloat(
     "Vector (RGB)"=&Multiply_VectorByFloat(
      Vector=&MakeVectorFromLinearColorRGB(Color=<rgba 1 0.0958 0.3904 1>)
      Float=&FloatFromCurve(FloatCurve=DataInterface() {propertyValues {Curve {keys [<at 0 1>, <at 0.1234 2>, {arriveTangent=-1.1407, interpMode=RCIM_Cubic, leaveTangent=-1.1407, time=0.3119, value=1.5}, <at 1 1>]}}})
     )
     "Float (Alpha)"=&FloatFromCurve(
      FloatCurve=DataInterface() {propertyValues {Curve {
         defaultValue=0
         keys [
          {interpMode=RCIM_Cubic, tangentMode=RCTM_User, time=0, value=0}
          {interpMode=RCIM_Cubic, tangentMode=RCTM_User, time=0.5, value=1}
          {tangentMode=RCTM_User, time=1, value=0}
         ]
        }}}
      Scale Curve=0.35
     )
    )) {isSetParameters=true}
   /SetVariables(
    Particles.HeartOutlineColor:LinearColor=&MultiplyLinearColor(
     Linear Color=$Particles.Color
     Scale Factor=&MakeVector4FromVectorAndFloat("Vector (RGB)"=&MakeVectorFromLinearColorRGB(Color=<rgba 1.8455 2.4212 5 1>), "Float (Alpha)"=0.8)
    )
    Particles.HeartOutlineScale:Vector3f=&VectorFromFloat(Value=&FloatFromCurve(
      FloatCurve=DataInterface() {propertyValues {Curve {keys [
          {arriveTangent=2.3927, interpMode=RCIM_Cubic, leaveTangent=2.3927, leaveTangentWeight=0.3333, tangentMode=RCTM_User, time=0, value=0.125}
          {arriveTangentWeight=0.3333, interpMode=RCIM_Cubic, tangentMode=RCTM_User, time=1, value=1}
         ]}}}
      Scale Curve=3
     ))
   ) {isSetParameters=true}
  ]
  rendererClasses [
   &NiagaraMeshRendererProperties
   &NiagaraSpriteRendererProperties
   &NiagaraLightRendererProperties
  ]
  renderers [
   {rendererClass=&NiagaraMeshRendererProperties, rendererIndex=0}
   {rendererClass=&NiagaraSpriteRendererProperties, rendererIndex=1}
   {rendererClass=&NiagaraLightRendererProperties, rendererIndex=2}
   {rendererClass=&NiagaraSpriteRendererProperties, rendererIndex=3}
   {rendererClass=&NiagaraSpriteRendererProperties, rendererIndex=4}
   {rendererClass=&NiagaraSpriteRendererProperties, rendererIndex=5}
   {rendererClass=&NiagaraSpriteRendererProperties, rendererIndex=6}
   {rendererClass=&NiagaraMeshRendererProperties, rendererIndex=7}
   {rendererClass=&NiagaraMeshRendererProperties, rendererIndex=8}
  ]
 }
 airflow {
  emitterUpdateScript [
   &EmitterState()
   &SpawnBurst_Instantaneous(Spawn Count=2, Use Spawn Probability=false, UseLoopCountLimit=false)
  ]
  particleSpawnScript [
   &FN_SetSpawnGroup(Number of Spawn Groups=2)
   &InitializeParticle(
    Color=<rgba 5 2.7865 4.2226 0.5>
    Lifetime=&MaskFloatBySpawnGroup(Masked Float=0.8, Passthrough Float=0.8)
    Mesh Scale=&Multiply_VectorByFloat(
     Vector=&RandomRangeVector(Minimum=<1 1 0.8>, Maximum=<1 1 1>)
     Float=&RandomRangeFloat(Minimum=1.5, Maximum=2.2)
    )
    Mesh Scale Mode=::NewEnumerator4
    Position Offset=<-22.5 -4 90>
    UsePositionOffset=true
    Write Material Random=true
   )
   &InitialMeshOrientation(
    Mesh Orientation Mode=::NewEnumerator1
    Rotation=&MaskVectorBySpawnGroup(Masked Vector=<0 -0.06 0.055>, Passthrough Vector=<0 0.04 -0.07>)
    Use Rotation Vector=true
   )
  ]
  particleUpdateScript [
   &ParticleState()
   &DynamicMaterialParameters(
    Index 0 Param 1=1
    Index 0 Param 4=1
    Module.Index0Float0Write=true
    Module.Index0Float1Write=true
    Module.Index0Float2Write=true
    Module.Index0Float3Write=true
   )
   &ScaleMeshSize(Scale Factor=&Multiply_VectorByFloat(
     Vector=<1 1 1>
     Float=&FloatFromCurve(
      FloatCurve=DataInterface() {propertyValues {Curve {keys [
          {arriveTangent=2.2546, interpMode=RCIM_Cubic, leaveTangent=2.2546, tangentMode=RCTM_User, time=0, value=0.3}
          {arriveTangent=0.9995, interpMode=RCIM_Cubic, leaveTangent=0.9995, tangentMode=RCTM_User, time=0.3, value=0.8}
          {arriveTangent=-0.0405, interpMode=RCIM_Cubic, leaveTangent=-0.0405, tangentMode=RCTM_User, time=1, value=1}
         ]}}}
      Scale Curve=0.8
     )
    ))
   &UpdateMeshOrientation(Rotation Rate=&RandomRangeFloat(Minimum=1.6, Maximum=2.4))
   &ScaleColor(
    Scale Alpha=&FloatFromCurve(FloatCurve=DataInterface() {propertyValues {Curve {
        defaultValue=0
        keys [
         {arriveTangent=9.0725, interpMode=RCIM_Cubic, leaveTangent=9.0725, tangentMode=RCTM_User, time=0, value=0}
         {arriveTangent=0.1284, interpMode=RCIM_Cubic, leaveTangent=0.1284, tangentMode=RCTM_User, time=0.1408, value=0.9847}
         {arriveTangent=-0.3715, interpMode=RCIM_Cubic, leaveTangent=-0.3715, tangentMode=RCTM_User, time=0.8, value=0}
        ]
       }}})
    ScaleA=true
    ScaleRGB=true
   )
   /SetVariables(
    Particles.Dark Color:LinearColor=&MakeLinearColorFromVectorAndFloat(
     "Vector (RGB)"=&MakeVectorFromLinearColorRGB(Color=<rgba 2 0.7149 1.4071 2>)
     "Float (Alpha)"=&Multiply_Float(
      A=1
      B=&FloatFromCurve(FloatCurve=DataInterface() {propertyValues {Curve {
          defaultValue=0
          keys [
           {arriveTangent=9.0725, interpMode=RCIM_Cubic, leaveTangent=9.0725, tangentMode=RCTM_User, time=0, value=0}
           {arriveTangent=-0.0182, interpMode=RCIM_Cubic, leaveTangent=-0.0182, tangentMode=RCTM_User, time=0.1408, value=0.9847}
           {arriveTangent=-0.268, interpMode=RCIM_Cubic, leaveTangent=-0.268, tangentMode=RCTM_User, time=1, value=0}
          ]
         }}})
     )
    )
    Particles.DarkDP:Vector4f=&MakeVector4(
     X=&FloatFromCurve(FloatCurve=DataInterface() {propertyValues {Curve {
         defaultValue=1
         keys [{interpMode=RCIM_Cubic, tangentMode=RCTM_User, time=0, value=1}, {interpMode=RCIM_Cubic, tangentMode=RCTM_User, time=1, value=0}]
        }}})
     Y=1
     Z=0
     W=0
    )
    Particles.DarkScale:Vector3f=&Multiply_Vector(A=$Particles.Scale, B=<1 1 2.0603>)
   ) {isSetParameters=true}
   /SetVariables(
    Particles.GlowColor=&MakeLinearColorFromVectorAndFloat(
     "Vector (RGB)"=&MakeVectorFromLinearColorRGB(Color=<rgba 1 0.3575 0.7036 2>)
     "Float (Alpha)"=&Multiply_Float(
      A=0.4
      B=&FloatFromCurve(FloatCurve=DataInterface() {propertyValues {Curve {
          defaultValue=0
          keys [
           {arriveTangent=9.0725, interpMode=RCIM_Cubic, leaveTangent=9.0725, tangentMode=RCTM_User, time=0, value=0}
           {arriveTangent=-0.0182, interpMode=RCIM_Cubic, leaveTangent=-0.0182, tangentMode=RCTM_User, time=0.1408, value=0.9847}
           {arriveTangent=-0.268, interpMode=RCIM_Cubic, leaveTangent=-0.268, tangentMode=RCTM_User, time=0.925, value=0}
          ]
         }}})
     )
    )
    Particles.GlowScale:Vector3f=&Multiply_Vector(A=$Particles.Scale, B=<1 1 1.2>)
   ) {isSetParameters=true}
  ]
  rendererClasses [&NiagaraMeshRendererProperties]
  renderers [
   {rendererClass=&NiagaraMeshRendererProperties, rendererIndex=0}
   {rendererClass=&NiagaraMeshRendererProperties, rendererIndex=1}
   {rendererClass=&NiagaraMeshRendererProperties, rendererIndex=2}
  ]
 }
}
summary {
 systemName=NS_Emote_PearlTaffyOpera_Burst
 emitters [
  {
   emitterName=SimpleSpriteBurst
   rendererClasses [
    &NiagaraMeshRendererProperties
    &NiagaraSpriteRendererProperties
    &NiagaraLightRendererProperties
   ]
  }
  {emitterName=airflow, rendererClasses [&NiagaraMeshRendererProperties]}
  {emitterName=AirflowDust, rendererClasses [&NiagaraSpriteRendererProperties]}
 ]
 userVariables [User.AirflowDustSpawnRate {type=NiagaraFloat, defaultValue=0, description=""}]
}
