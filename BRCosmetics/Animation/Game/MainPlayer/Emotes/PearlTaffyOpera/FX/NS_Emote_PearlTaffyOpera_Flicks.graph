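# Niagara system for the flick effects of the PearlTaffyOpera player emote, drawing pink-tinted sprite sparks with glow and sparkle colours.
# In Flicks, each change of User.BurstIndex triggers two spawn groups with per-index counts.
# Those particles spawn at skeletal-mesh bones or sockets chosen by partition, with size and speed varying by group.
# Minimal ignores the index, firing four timed bursts of glowing meshes and sprites from a fixed offset.
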NiagaraSystem /BRCosmetics/Animation/Game/MainPlayer/Emotes/PearlTaffyOpera/FX/NS_Emote_PearlTaffyOpera_Flicks
dependencies {
 usedDataInterfaces [&NiagaraDataInterfaceCurve, &NiagaraDataInterfaceSkeletalMesh]
 usedDynamicInputs []
 usedModules [
  &self:SystemSpawnScript.SystemScriptSource.SystemScriptGraph.NiagaraNodeAssignment_0.SetVariables_#1
  &SystemState
 ]
}
emitters {
 Flicks {
  emitterUpdateScript [
   &EmitterState()
   /Game/Effects/Niagara/Modules/Logic/UpdateValueStateChange(InputIntParam:NiagaraInt32=&MakeIntFromFloat(Float=$User.BurstIndex, Conversion Mode=::NewEnumerator1))
   &self:TriggerSpawn(
    Spawn Count:NiagaraInt32=&MergeHelper_IntSwitch(
     0=8
     1=6
     2=4
     3=4
     4=6
     5=8
     6=7
     7=9
     Index=&MakeIntFromFloat(Float=$User.BurstIndex, Conversion Mode=::NewEnumerator1)
     Num=::NewEnumerator6
    )
    Spawn Group:NiagaraInt32=&MakeIntFromFloat(Float=0, Conversion Mode=::NewEnumerator1)
    Trigger:NiagaraBool=$Emitter.UpdateValueStateChange.Changed
   ) {module=TriggerSpawn}
   &self:TriggerSpawn(
    Spawn Count:NiagaraInt32=&MergeHelper_IntSwitch(
     0=10
     1=12
     2=8
     3=9
     4=12
     5=22
     6=16
     7=28
     Index=&MakeIntFromFloat(Float=$User.BurstIndex, Conversion Mode=::NewEnumerator1)
     Num=::NewEnumerator6
    )
    Spawn Group:NiagaraInt32=&MakeIntFromFloat(Float=1, Conversion Mode=::NewEnumerator1)
    Trigger:NiagaraBool=$Emitter.UpdateValueStateChange.Changed
   ) {module=TriggerSpawn001}
  ]
  particleSpawnScript [
   /SetVariables(Particles.VisibilityTag=&SpawnGroup()) {isSetParameters=true}
   &PartitionParticles()
   &InitializeParticle(
    Color=<rgba 2 1.1458 1.1458 1>
    Lifetime Max=0.8
    Lifetime Min=0.3
    Lifetime Mode=::NewEnumerator1
    Sprite Size Mode=::NewEnumerator3
    Uniform Sprite Size=&MergeHelper_FloatSwitch(
     0=&RandomRangeFloat(Minimum=25, Maximum=40)
     1=&RandomRangeFloat(Minimum=50, Maximum=85)
     2=?
     3=?
     Index=&SpawnGroup()
     Num=::NewEnumerator0
    )
    UsePositionOffset=false
    Write Material Random=true
   )
   &SkeletalMeshLocation(
    Bone / Socket Index=$Particles.PartitionParticles.Partition
    Bone Sampling Mode=::NewEnumerator4
    MaskSpawnGroup=false
    Position Sampling=::NewEnumerator1
    Skeletal Mesh=$System.Skeletal Mesh Reader
   )
   &AddVelocity(
    Cone Angle=60
    Cone Axis=&Multiply_VectorByFloat(
     Vector=&ReturnMeshOrientationAxis(Mesh Orientation Axis=::YAxis, Mesh Orientation Quat=$Particles.MeshOrientation)
     Float=&MergeHelper_FloatSwitch(0=1, 1=1, 2=?, 3=?, Index=&SpawnGroup(), Num=::NewEnumerator0)
    )
    Module.UseConeFalloff=false
    Module.UseVelDistribution=false
    Override Seed=false
    Velocity Mode=::NewEnumerator2
    Velocity Speed=&MergeHelper_FloatSwitch(
     0=&RandomRangeFloat(Minimum=45, Maximum=180)
     1=&RandomRangeFloat(Minimum=80, Maximum=225)
     2=?
     3=?
     Index=&SpawnGroup()
     Num=::NewEnumerator0
    )
   )
   &CameraOffset(Camera Offset Amount=&RandomRangeFloat(Minimum=10, Maximum=35))
  ]
  particleUpdateScript [
   &ParticleState()
   &ScaleSpriteSize(
    Uniform Curve Scale=&MergeHelper_FloatSwitch(
     0=1
     1=&FloatFromCurve(
      FloatCurve=DataInterface() {propertyValues {Curve {keys [<at 0 0 RCIM_Constant>,<at 0.4 1 RCIM_Constant>,<at 1 1 RCIM_Constant>]}}}
      CurveIndex=$Emitter.NormalizedLoopAge
     )
     2=?
     3=?
     Index=$Particles.PartitionParticles.Partition
     Num=::NewEnumerator0
    )
    Uniform Curve Sprite Scale=DataInterface() {propertyValues {Curve {
       defaultValue=1
       keys [
        {arriveTangent=9.0725, interpMode=RCIM_Cubic, leaveTangent=9.0725, tangentMode=RCTM_User, time=0, value=0}
        {arriveTangent=-0.0031, interpMode=RCIM_Cubic, leaveTangent=-0.0031, tangentMode=RCTM_User, time=0.1408, value=0.9847}
        {arriveTangent=-2.0324, interpMode=RCIM_Cubic, leaveTangent=-2.0324, tangentMode=RCTM_User, time=1, value=0}
       ]
      }}}
   )
   &ScaleColor(
    Scale Alpha=&FloatFromCurve(FloatCurve=DataInterface() {propertyValues {Curve {
        defaultValue=1
        keys [
         {arriveTangent=9.0725, interpMode=RCIM_Cubic, leaveTangent=9.0725, tangentMode=RCTM_User, time=0, value=0}
         {arriveTangent=-2.5249, interpMode=RCIM_Cubic, leaveTangent=-2.5249, tangentMode=RCTM_User, time=0.1408, value=0.9847}
         {arriveTangent=-0.0405, interpMode=RCIM_Cubic, leaveTangent=-0.0405, tangentMode=RCTM_User, time=1, value=0}
        ]
       }}})
    ScaleA=true
    ScaleRGB=true
   )
   &Drag(Drag=&RandomRangeFloat(Minimum=0.8, Maximum=1.2), Use Linear Drag=true, Use Rotational Drag=true)
   &SolveForcesAndVelocity(Clamp Velocity=false, Limit Acceleration=false)
   /SetVariables(
    Particles.GlowColor=&MultiplyLinearColor(
     Linear Color=<rgba 0.5 0.5 0.5 1>
     Scale Factor=&MakeVector4FromVectorAndFloat(
      "Vector (RGB)"=&MakeVectorFromLinearColorRGB(Color=&RandomRangeLinearColor(Minimum=<rgba 0.9639 1.7703 5 1>, Maximum=<rgba 5 1.2208 2.3366 1>))
      "Float (Alpha)"=&RandomRangeFloat(Minimum=0.1, Maximum=0.225)
     )
    )
    Particles.GlowSize=&Multiply_Vector2D_ByFloat(Float=2, Vector2D=$Particles.SpriteSize)
   ) {isSetParameters=true}
   /SetVariables(Particles.SparkleColor:LinearColor=&MakeLinearColorFromVectorAndFloat(
     "Vector (RGB)"=&MakeVectorFromLinearColorRGB(Color=&RandomRangeLinearColor(Minimum=<rgba 0.93 0.15 1 1>, Maximum=<rgba 1 0.2548 0.7529 1>))
     "Float (Alpha)"=&FloatFromCurve(
      FloatCurve=DataInterface() {propertyValues {Curve {
         defaultValue=1
         keys [
          {arriveTangent=9.0725, interpMode=RCIM_Cubic, leaveTangent=9.0725, tangentMode=RCTM_User, time=0, value=0}
          {arriveTangent=-0.2024, interpMode=RCIM_Cubic, leaveTangent=-0.2024, tangentMode=RCTM_User, time=0.1408, value=0.9847}
          {arriveTangent=-1.809, interpMode=RCIM_Cubic, leaveTangent=-1.809, tangentMode=RCTM_User, time=1, value=0}
         ]
        }}}
      Scale Curve=&Waveform("Waveform [1]"=::NewEnumerator2, "[1] Frequency"=5, "[1] Amplitude Min/Max"=<0.3 1>)
     )
    )) {isSetParameters=true}
  ]
  rendererClasses [&NiagaraSpriteRendererProperties]
  renderers [
   {rendererClass=&NiagaraSpriteRendererProperties}
   {rendererClass=&NiagaraSpriteRendererProperties, rendererIndex=1}
   {rendererClass=&NiagaraSpriteRendererProperties, rendererIndex=2}
  ]
 }
 Minimal {
  emitterUpdateScript [
   &EmitterState()
   &SpawnBurst_Instantaneous(Use Spawn Probability=false, UseLoopCountLimit=false)
   &SpawnBurst_Instantaneous(Spawn Time=0.41, Use Spawn Probability=false, UseLoopCountLimit=false) {module=SpawnBurst_Instantaneous001}
   &SpawnBurst_Instantaneous(Spawn Time=0.76, Use Spawn Probability=false, UseLoopCountLimit=false) {module=SpawnBurst_Instantaneous002}
   &SpawnBurst_Instantaneous(Spawn Time=1.11, Use Spawn Probability=false, UseLoopCountLimit=false) {module=SpawnBurst_Instantaneous003}
  ]
  particleSpawnScript [
   &InitializeParticle(
    Color=<rgba 1 0.3954 0.5829 1>
    Lifetime=0.6
    Mesh Scale Mode=::NewEnumerator3
    Mesh Uniform Scale=0.5
    Position Offset=<4 15 110>
    UsePositionOffset=true
    Write Material Random=true
   )
   &AddVelocity(
    Cone Axis=<-0.25 1 0>
    Module.UseConeFalloff=false
    Module.UseVelDistribution=false
    Override Seed=false
    Velocity Mode=::NewEnumerator2
    Velocity Speed=65
   )
   &InitialMeshOrientation(Mesh Orientation Mode=::NewEnumerator1, Rotation=<0 0 0.05>, Use Rotation Vector=true)
   &CameraOffset(Camera Offset Amount=15)
  ]
  particleUpdateScript [
   &ParticleState()
   &ScaleColor(
    Scale Alpha=&FloatFromCurve(FloatCurve=DataInterface() {propertyValues {Curve {
        defaultValue=0
        keys [
         {arriveTangent=9.0725, interpMode=RCIM_Cubic, leaveTangent=9.0725, tangentMode=RCTM_User, time=0, value=0}
         {arriveTangent=-2.5249, interpMode=RCIM_Cubic, leaveTangent=-2.5249, tangentMode=RCTM_User, time=0.1408, value=0.9847}
         {arriveTangent=-0.0405, interpMode=RCIM_Cubic, leaveTangent=-0.0405, tangentMode=RCTM_User, time=1, value=0}
        ]
       }}})
    Scale RGB=&VectorFromFloat(Value=&FloatFromCurve(FloatCurve=DataInterface() {propertyValues {Curve {keys [<at 0 1>, <at 0.1234 2>, {arriveTangent=-1.1407, interpMode=RCIM_Cubic, leaveTangent=-1.1407, time=0.3119, value=1.5}, <at 1 1>]}}}))
    ScaleA=true
    ScaleRGB=true
   )
   &ScaleMeshSize(Scale Factor=&VectorFromFloat(Value=&FloatFromCurve(FloatCurve=DataInterface() {propertyValues {Curve {defaultValue=1, keys [<at 0 0 RCIM_Cubic>,<at 0.1 1 RCIM_Cubic>,<at 1 0 RCIM_Cubic>]}}})))
   &DynamicMaterialParameters(
    Index 0 Param 1=&FloatFromCurve(
     FloatCurve=DataInterface() {propertyValues {Curve {
        defaultValue=1
        keys [
         {arriveTangent=9.0725, interpMode=RCIM_Cubic, leaveTangent=9.0725, tangentMode=RCTM_User, time=0, value=0}
         {arriveTangent=-2.5249, interpMode=RCIM_Cubic, leaveTangent=-2.5249, tangentMode=RCTM_User, time=0.1408, value=0.9847}
         {arriveTangent=-0.0405, interpMode=RCIM_Cubic, leaveTangent=-0.0405, tangentMode=RCTM_User, time=1, value=0}
        ]
       }}}
     Scale Curve=3
    )
    Index 0 Param 3=&FloatFromCurve(FloatCurve=DataInterface() {propertyValues {Curve {defaultValue=1, keys [{tangentMode=RCTM_User, time=0, value=0}, {tangentMode=RCTM_User, time=1, value=1}]}}})
    Module.Index0Float0Write=true
    Module.Index0Float1Write=false
    Module.Index0Float2Write=true
    Module.Index0Float3Write=false
   )
   &Drag(Drag=2, Use Linear Drag=true, Use Rotational Drag=true)
   &SolveForcesAndVelocity(Clamp Velocity=false, Limit Acceleration=false)
   /SetVariables(
    Particles.GlowColor=&MultiplyLinearColor(
     Linear Color=$Particles.Color
     Scale Factor=&MakeVector4FromVectorAndFloat("Vector (RGB)"=<1 1 1>, "Float (Alpha)"=&RandomRangeFloat(Minimum=0.075, Maximum=0.15))
    )
    Particles.GlowSize=&Vector2DFromFloat(Value=&FloatFromCurve(
      FloatCurve=DataInterface() {propertyValues {Curve {
         defaultValue=1
         keys [
          {tangentMode=RCTM_User, time=0, value=1}
          {arriveTangent=-0.0405, interpMode=RCIM_Cubic, leaveTangent=-0.0405, tangentMode=RCTM_User, time=0.7009, value=1}
          {arriveTangent=-8.0089, interpMode=RCIM_Cubic, leaveTangent=-8.0089, tangentMode=RCTM_User, time=1, value=0}
         ]
        }}}
      Scale Curve=60
     ))
   ) {isSetParameters=true}
  ]
  rendererClasses [&NiagaraMeshRendererProperties, &NiagaraSpriteRendererProperties]
  renderers [
   {rendererClass=&NiagaraMeshRendererProperties, rendererIndex=0}
   {rendererClass=&NiagaraSpriteRendererProperties, rendererIndex=1}
  ]
 }
}
summary {
 systemName=NS_Emote_PearlTaffyOpera_Flicks
 emitters [
  {emitterName=Flicks, rendererClasses [&NiagaraSpriteRendererProperties]}
  {emitterName=Minimal, rendererClasses [&NiagaraMeshRendererProperties, &NiagaraSpriteRendererProperties]}
 ]
 userVariables [User.BurstIndex {type=NiagaraFloat, defaultValue=0, description=""}]
}
